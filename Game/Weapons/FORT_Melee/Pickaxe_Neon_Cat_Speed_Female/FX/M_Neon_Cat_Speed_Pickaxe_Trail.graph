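# Unlit, additive, two-sided Niagara ribbon material for the trail of the Neon Cat Speed pickaxe.
# Texture RGB supplies line, pattern and fill layers, each with its own intensity and length-wise fade exponent.
# Their normalized sum, with vertical fade and tip boost, multiplies particle color into an emissive-adjustment function.
# A tip fade broken up by the pattern channel forms a mask, adding view-angle and depth fades except on mobile.
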Material /Game/Weapons/FORT_Melee/Pickaxe_Neon_Cat_Speed_Female/FX/M_Neon_Cat_Speed_Pickaxe_Trail
graph {
 properties {
  bUsedWithNiagaraRibbons=true
  blendMode=BLEND_Additive
  pixelDepthOffsetMode=PDOM_Legacy
  shadingModel=MSM_Unlit
  twoSided=true
  usedShadingModels=Unlit
 }
 nodes [
  n1 = ScalarParameter() {defaultValue=0.35, parameterName=Hue Shift Intensity, sliderMax=1}
  n2 = ComponentMask(None=TextureCoordinate()) {r=true}
  n3 = ScalarParameter() {defaultValue=0.08, parameterName=Hue Shift Offset}
  n4 = Saturate(None=Divide(A=Subtract(A=@n2, B=@n3), B=ScalarParameter() {defaultValue=0.16, parameterName=Hue Shift Softness}))
  n5 = ParticleColor()
  n6 = ScalarParameter() {defaultValue=1, parameterName=switch}
  LinearInterpolate(
   A=MaterialFunctionCall(Hue Shift Percentage=Multiply(A=@n1, B=@n4), Texture=VectorParameter() {parameterName=Main Color, defaultValue=<rgba 0.36 8.712 9 3>}.RGB) {materialFunction=&HueShift}.Result
   B=@n5.RGB
   Alpha=Round(None=Saturate(None=@n6))
  )
  n7 = TextureCoordinate()
  n8 = OneMinus(None=Multiply(A=ComponentMask(None=@n7) {g=true}, B=ScalarParameter() {defaultValue=0.5, parameterName=Vertical Fade Amount}))
  n9 = TextureSample(UVs=TextureCoordinate() {coordinateIndex=1}) {texture=&/T_Neon_Cat_Speed_Pickaxe}
  n10 = ScalarParameter() {defaultValue=1, parameterName=Line Intensity}
  n11 = ScalarParameter() {defaultValue=0.2, parameterName=Pattern Intensity}
  n12 = ScalarParameter() {defaultValue=0.025, parameterName=Fill Intensity}
  n13 = Saturate(None=MaterialFunctionCall(R=@n10, G=@n11, B=@n12) {materialFunction=&AppendMany}.RGB)
  n14 = Power(
   Base=OneMinus(None=@n2)
   Exp=MaterialFunctionCall(
    R=ScalarParameter() {defaultValue=2, parameterName=Line Fade Exponent}
    G=ScalarParameter() {defaultValue=0.5, parameterName=Pattern Fade Exponent}
    B=ScalarParameter() {defaultValue=1.1, parameterName=Fill Fade Exponent}
   ) {materialFunction=&AppendMany}.RGB
  )
  n15 = TextureCoordinate()
  n16 = Saturate(None=Divide(A=ComponentMask(None=@n15) {r=true}, B=ScalarParameter() {defaultValue=1, parameterName=Tip Fill Boost}))
  n17 = Divide(A=Power(Base=Multiply(A=@n9.RGB, B=@n13), Exp=Divide(B=@n14)), B=Multiply(A=@n16, B=@n16))
  n18 = Divide(
   A=MaterialFunctionCall(f3=Multiply(A=@n8, B=@n17)) {materialFunction=&AddComponents}.f3
   B=Add(A=Add(A=@n10, B=@n11), B=@n12)
  )
  MaterialFunctionCall(Emissive=Multiply(A=@n5.RGB, B=Saturate(None=@n18))) {materialFunction=&MF_FX_EmissiveAdjustments}
  n19 = Abs(None=DotProduct(A=VertexNormalWS(), B=CameraVectorWS()))
  n20 = Saturate(None=Divide(A=@n2, B=ScalarParameter() {defaultValue=0.1, parameterName=Tip Fade}) {constB=0.2})
  n21 = OneMinus(None=@n20)
  n22 = OneMinus(None=@n9.G)
  n23 = Multiply(A=@n20, B=OneMinus(None=Multiply(A=@n21, B=@n22)))
  n24 = ScalarParameter() {defaultValue=22, parameterName=Depth Fade Distance}
  ShadingPathSwitch(Default=Multiply(A=Saturate(None=@n19), B=DepthFade(Opacity=@n23, FadeDistance=@n24)), Mobile=@n23) {desc=Migrated from FeatureLevelSwitch}
 ]
}
